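# Spot ambient one-shot cue for the duck yacht horn, layering a close and a distant recording.
# A mixer blends the two waves, each behind its own attenuation node with logarithmic falloff, distance low-pass, listener focus and occlusion, so each layer fades over its own range.
# The root modulator adds no pitch or volume variation.
# The cue also sends to the exterior ambient source bus.
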SoundCue /Hera/Environment/Blueprints/DuckHorn/Assets/audio/Spot_Duck_Yatch_Horn_Cue
root=Modulator_0
nodes [
 {
  id=Attenuation_1
  children [WavePlayer_4]
  properties {
   bOverrideAttenuation=true
   attenuationOverrides {
    bAttenuateWithLPF=true
    bEnableListenerFocus=true
    bEnableOcclusion=true
    bEnableReverbSend=false
    distanceAlgorithm=Logarithmic
    falloffDistance=40000
    lPFRadiusMax=20000
    lPFRadiusMin=10000
    nonFocusVolumeAttenuation=0.75
    occlusionLowPassFilterFrequency=5000
    reverbWetLevelMin=0.404
    attenuationShapeExtents {x=5000}
    customAttenuationCurve {editorCurveData {keys [
       <at 0 1 RCIM_Cubic>
       {arriveTangent=-1.6482, interpMode=RCIM_Cubic, leaveTangent=-1.6482, tangentMode=RCTM_User, time=0.7384, value=0.4106}
       <at 1 0>
      ]}}
   }
  }
 }
 {
  id=Attenuation_2
  children [WavePlayer_3]
  properties {
   bOverrideAttenuation=true
   attenuationOverrides {
    bAttenuateWithLPF=true
    bEnableFocusInterpolation=true
    bEnableListenerFocus=true
    bEnableOcclusion=true
    distanceAlgorithm=Logarithmic
    falloffDistance=30000
    focusVolumeAttenuation=1.2
    lPFFrequencyAtMax=5000
    lPFRadiusMax=10000
    nonFocusVolumeAttenuation=0.75
    nonSpatializedRadiusEnd=200
    nonSpatializedRadiusStart=1000
    occlusionLowPassFilterFrequency=5000
    occlusionTraceChannel=ECC_GameTraceChannel11
    customAttenuationCurve {editorCurveData {keys [
       <at 0 1>
       {arriveTangent=-2.4936, interpMode=RCIM_Cubic, leaveTangent=-2.4936, tangentMode=RCTM_User, time=0.0384, value=0.9588}
       {arriveTangent=-3.3235, interpMode=RCIM_Cubic, leaveTangent=-3.3235, time=0.0503, value=0.2422}
       <at 0.3168 0.0336>
       {arriveTangent=-3.4272, interpMode=RCIM_Cubic, leaveTangent=-3.4272, tangentMode=RCTM_User, time=1, value=0}
      ]}}
   }
  }
 }
 {id=Mixer_0, children [Attenuation_1, Attenuation_2], properties {inputVolume [2,1.6]}}
 {id=Modulator_0, children [Mixer_0], properties {pitchMax=1, pitchMin=1, volumeMax=1, volumeMin=1}}
 {id=WavePlayer_3, properties {soundWaveAssetPtr=&/Duck_Yatch_Horn_Close}}
 {id=WavePlayer_4, properties {soundWaveAssetPtr=&/Duck_Yatch_Horn_Distant}}
]
properties {
 bOverrideAttenuation=true
 maxDistance=45000
 soundClassObject=&Fort_Ambient_OneShots
 volumeMultiplier=0.46
 attenuationOverrides {
  bEnableOcclusion=true
  falloffDistance=6000
  focusVolumeAttenuation=1.2
  lPFFrequencyAtMax=2000
  lPFRadiusMin=2000
  nonFocusVolumeAttenuation=0.4
  occlusionLowPassFilterFrequency=1000
  occlusionTraceChannel=ECC_GameTraceChannel11
  occlusionVolumeAttenuation=0.5
  stereoSpread=2000
  attenuationShapeExtents {x=1000}
 }
 busSends [{
   audioBus=null
   maxSendDistance=4000
   minSendLevel=0.6
   sendLevel=1
   soundSourceBus=&AmbientAudio_Exterior_SourceBus
  }]
}
